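; Preview actor that displays a skeletal mesh, apparently for a Dataflow preview scene.
; Its construction script assigns the SkeletalMesh variable to the preview component and uses the component bounds to offset it vertically, placing its bottom slightly above the scene floor, assumed at world Z zero.
; The offset teleports without sweeping, then physics simulation is switched off.
; BeginPlay, overlap and Tick events are empty.
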(event EventBeginPlay)

(event Collision|EventActorBeginOverlap
  (params (OtherActor :type Actor%20Object%20Reference)))

(event EventTick
  (params (DeltaSeconds :type Float%20%28single-precision%29)))

(fn ConstructionScript
  (bind (_origin _boxextent) (Collision|GetComponentBounds (Variables|Default|GetPreviewComponent)))
  (bind _previewcomponent (Variables|Default|GetPreviewComponent))
  (Components|SkeletalMesh|SetSkeletalMeshAsset _previewcomponent (Variables|Default|GetSkeletalMesh))
  ;[1 1 1] Move the preview component so it is above the preview scene floor slightly
  (Transformation|AddWorldOffset _previewcomponent (Math|Vector|MakeVector 0.0 0.0 (+ (* (- (.z _origin) (.z _boxextent)) -1.0) 1e+01)) false true)
  (Physics|SetSimulatePhysics _previewcomponent false))
